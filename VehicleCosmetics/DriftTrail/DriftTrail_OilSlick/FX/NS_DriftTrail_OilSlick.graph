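# Niagara vehicle cosmetic drift trail that makes oil slick effects at the car's wheels.
# Three emitters supply rotating wheel-wrap meshes, spawn-grouped cone meshes and sprites with gravity-pulled drops, and two ribbon trails layered as oil, waves and tire prints.
# Particles attach to skeletal mesh sockets, use per-wheel FN_DisableTireFX masking, and switch offsets by game mode.
# Painted-variant override modules are present but disabled, leaving fixed dark oil colors.
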NiagaraSystem /VehicleCosmetics/DriftTrail/DriftTrail_OilSlick/FX/NS_DriftTrail_OilSlick
dependencies {
 usedDataInterfaces [&NiagaraDataInterfaceCurve, &NiagaraDataInterfaceColorCurve, &NiagaraDataInterfaceSkeletalMesh]
 usedDynamicInputs []
 usedModules [
  &self:SystemSpawnScript.SystemScriptSource.SystemScriptGraph.NiagaraNodeAssignment_0.SetVariables_#1
  &FN_UnpackVehicleFlags
  /VehicleCosmetics/FX/Niagara/Modules/VC_SetVehicleSpeedAttribs
  &SystemState
 ]
}
emitters {
 NE_Cone-Particles001 {
  emitterUpdateScript [
   &EmitterState(
    Inactive Response=::NewEnumerator1
    Life Cycle Mode=::NewEnumerator1
    Loop Behavior=::NewEnumerator1
    Loop Duration Mode=::NewEnumerator1
    UseLoopDelay=true
   )
   &SpawnRate(SpawnGroup=3, SpawnRate=40, Use Spawn Probability=false) {module=SpawnRate003}
   &SpawnRate(SpawnGroup=2, SpawnRate=50, Use Spawn Probability=false) {module=SpawnRate002}
   &SpawnRate(SpawnRate=10, Use Spawn Probability=false) {module=SpawnRate001}
   &SpawnRate(SpawnGroup=1, SpawnRate=10, Use Spawn Probability=false)
  ]
  particleSpawnScript [
   &VC_PaintedVariantsOverride(
    DefaultColor=<rgba 0 0 0 10>
    R1 DefaultColor=<rgba 0.0144 0.0144 0.026 2>
    R2 DefaultColor:LinearColor=<rgba 0 0 0 1>
    Use R1 Painted Overrides=true
    Use R2 Painted Overrides=true
   ) {enabled=false}
   &InitializeParticle(
    Color=<rgba 0 0 0 20>
    Lifetime=&MaskFloatBySpawnGroup(
     Masked Float=&RandomRangeFloat(Minimum=0.35, Maximum=0.4)
     Passthrough Float=&MaskFloatBySpawnGroup(Masked Float=&RandomRangeFloat(Minimum=0.5, Maximum=0.7), Passthrough Float=0.7, Spawn Group Mask=3)
     Spawn Group Mask=2
    )
    Mesh Scale=<1.4 2.5 5>
    Mesh Scale Mode=::NewEnumerator4
    Sprite Rotation Mode=::NewEnumerator2
    Sprite Size Mode=::NewEnumerator3
    Uniform Sprite Size=&MaskFloatBySpawnGroup(Masked Float=&RandomRangeFloat(Minimum=10, Maximum=20), Passthrough Float=30, Spawn Group Mask=2)
    UsePositionOffset=false
    Write Material Random=true
   )
   &PartitionParticles()
   &SkeletalMeshLocation(
    Bone / Socket Index=&Add_Integer(A=$Particles.PartitionParticles.Partition, B=4)
    Bone / Socket Sampling Mode=::NewEnumerator5
    Interpolate Mesh Sampling=false
    MaskSpawnGroup=false
    Mesh Sampling Type=::NewEnumerator6
    Sampled Position Offset=&AddVectorToPosition(
     Position=$User.offset
     Vector=&MaskVectorBySpawnGroup(
      Masked Vector=&MergeHelper_VectorSwitch(0=<-20 0 -40>, 1=<-20 0 -43>, 2=<-20 0 -40>, 3=?, Index=$System.GameModeID, Num=::NewEnumerator1)
      Passthrough Vector=&MaskVectorBySpawnGroup(
       Masked Vector=<-47 0 -32>
       Passthrough Vector=&MaskVectorBySpawnGroup(
        Masked Vector=&MergeHelper_VectorSwitch(0=<-20 0 -40>, 1=<-20 0 -43>, 2=<-20 0 -40>, 3=?, Index=$System.GameModeID, Num=::NewEnumerator1)
        Passthrough Vector=&MergeHelper_VectorSwitch(0=<-37 0 -35>, 1=<-30 0 -35>, 2=<-30 0 -35>, 3=?, Index=$System.GameModeID, Num=::NewEnumerator1)
       )
       Spawn Group Mask=2
      )
      Spawn Group Mask=1
     )
    )
    Skeletal Mesh=$System.SkeletalMesh
   )
   &AddVelocity(Velocity=&MaskVectorBySpawnGroup(
     Masked Vector=<-50 0 0>
     Passthrough Vector=&MaskVectorBySpawnGroup(
      Masked Vector=<-20 0 0>
      Passthrough Vector=&MaskVectorBySpawnGroup(Masked Vector=<-36 0 0>, Passthrough Vector=<0 0 0>, Spawn Group Mask=3)
      Spawn Group Mask=1
     )
    ))
   &AddVelocity(
    Cone Angle=25
    Cone Axis=<-1 0 1>
    Inner Cone Angle=80
    Module.UseConeFalloff=false
    Module.UseVelDistribution=false
    Override Seed=false
    Velocity Mode=::NewEnumerator2
    Velocity Speed=&MaskFloatBySpawnGroup(
     Masked Float=&FloatFromCurve(
      FloatCurve=DataInterface() {propertyValues {Curve {keys [{interpMode=RCIM_Cubic, tangentMode=RCTM_Break, time=0.5, value=1}, {interpMode=RCIM_Cubic, tangentMode=RCTM_Break, time=1, value=0}]}}}
      Scale Curve=280
     )
     Passthrough Float=0
     Spawn Group Mask=2
    )
   ) {module=AddVelocity001}
   &InitialMeshOrientation(
    Mesh Orientation Mode=::NewEnumerator6
    Orientation Quaternion:Quat4f=/Game/Effects/Niagara/Modules/MergeEmitters/MergeHelper_QuatSwitch(
     0:Quat4f=<-0.9157 -0.9271 1.0578 1>
     1:Quat4f=<-0.9157 -0.9271 0.9 1>
     2:Quat4f=<-0.9157 -0.9271 1.0578 1>
     3:Quat4f=?
     4:Quat4f=?
     5:Quat4f=?
     6:Quat4f=?
     7:Quat4f=?
     8:Quat4f=?
     9:Quat4f=?
     Index:NiagaraInt32=0
     Num:MergeHelper_SwitcherEnum=::NewEnumerator1
    )
    Use Rotation Vector=false
   )
  ]
  particleUpdateScript [
   &FN_DisableTireFX(
    Mask Index Attribute:NiagaraInt32=$Particles.PartitionParticles.Partition
    Mask Index FL=2
    Mask Index FR=3
    Mask Index RL=0
    Mask Index RR=1
   )
   &ParticleState()
   &ScaleMeshSize(Scale Factor=&MaskVectorBySpawnGroup(
     Masked Vector=&Multiply_VectorByFloat(
      Vector=<1 1 1>
      Float=&FloatFromCurve(FloatCurve=DataInterface() {propertyValues {Curve {defaultValue=1, keys [{tangentMode=RCTM_User, time=0, value=0.5}, {tangentMode=RCTM_User, time=1, value=1}]}}})
     )
     Passthrough Vector=<1 1 1>
    ))
   &ScaleSpriteSize(Uniform Curve Sprite Scale=DataInterface() {propertyValues {Curve {
       defaultValue=0
       keys [{interpMode=RCIM_Cubic, tangentMode=RCTM_User, time=0, value=1}, {interpMode=RCIM_Cubic, tangentMode=RCTM_User, time=1, value=0}]
      }}})
   &SpriteRotationRate()
   /SetVariables(
    Particles.DropsColor:LinearColor=&ColorFromCurve(ColorCurve=DataInterface() {propertyValues {
       AlphaCurve {keys [<at 0.0283 0 RCIM_Cubic>,<at 0.2492 10 RCIM_Cubic>,<at 1 0>]}
       BlueCurve {keys [<at 0 0.0052 RCIM_Cubic>,<at 1 0.01 RCIM_Cubic>]}
       GreenCurve {keys [<at 0 0.0052 RCIM_Cubic>,<at 1 0.0016 RCIM_Cubic>]}
       RedCurve {keys [<at 0 0.0052 RCIM_Cubic>,<at 1 0 RCIM_Cubic>]}
      }})
    Particles.FakeFillSize:Vector2f=&Multiply_Vector2D_ByFloat(
     Float=&FloatFromCurve(
      FloatCurve=DataInterface() {propertyValues {Curve {
         defaultValue=1
         keys [{interpMode=RCIM_Cubic, tangentMode=RCTM_Break, time=0, value=0}, {interpMode=RCIM_Cubic, tangentMode=RCTM_Break, time=1, value=1}]
        }}}
      Scale Curve=&MergeHelper_FloatSwitch(0=30, 1=30, 2=30, 3=?, Index=$System.GameModeID, Num=::NewEnumerator1)
     )
     Vector2D=<1 1>
    )
    Particles.OilBGColor:LinearColor=&MultiplyLinearColor(Linear Color=<rgba 0.0224 0.0224 0.06 15>, Scale Factor=<1 1 1 1>)
    Particles.RenderTag:NiagaraInt32=&SpawnGroup()
   ) {isSetParameters=true}
   &GravityForce(Gravity=&MaskVectorBySpawnGroup(Masked Vector=<0 0 -980>, Passthrough Vector=<0 0 0>, Spawn Group Mask=2))
   &SolveForcesAndVelocity(Clamp Velocity=false, Limit Acceleration=false)
   &ScaleColor(
    Scale Alpha=&FloatFromCurve(FloatCurve=DataInterface() {propertyValues {Curve {keys [{interpMode=RCIM_Cubic, tangentMode=RCTM_Break, time=0.5, value=1}, {interpMode=RCIM_Cubic, tangentMode=RCTM_Break, time=1, value=0}]}}})
    ScaleA=true
    ScaleRGB=true
   )
   &DynamicMaterialParameters(
    Index 0 Param 2=3
    Index 0 Param 3=&FloatFromCurve(FloatCurve=DataInterface() {propertyValues {Curve {
        defaultValue=1
        keys [
         {arriveTangent=-0.13, interpMode=RCIM_Cubic, leaveTangent=-0.13, tangentMode=RCTM_User, time=0, value=0}
         {interpMode=RCIM_Cubic, tangentMode=RCTM_User, time=1, value=9}
        ]
       }}})
    Module.Index0Float0Write=true
    Module.Index0Float1Write=true
    Module.Index0Float2Write=true
    Module.Index0Float3Write=true
   )
  ]
  rendererClasses [&NiagaraMeshRendererProperties, &NiagaraSpriteRendererProperties]
  renderers [
   {rendererClass=&NiagaraMeshRendererProperties, rendererIndex=0}
   {rendererClass=&NiagaraMeshRendererProperties, rendererIndex=1}
   {rendererClass=&NiagaraSpriteRendererProperties, rendererIndex=2}
   {rendererClass=&NiagaraSpriteRendererProperties, rendererIndex=3}
  ]
 }
 NE_OilTrail001 {
  emitterUpdateScript [
   &EmitterState(Inactive Response=::NewEnumerator1, Life Cycle Mode=::NewEnumerator1, Loop Duration=1, UseLoopDelay=true)
   &VC_RateQualityHandler(Epic Spawn Rate=30, High Spawn Rate=30, Low Spawn Rate=30, Use Custom Rate?=true)
   &VC_SpawnRibbon_v1(Number of Ribbons=2, SpawnRate=$Emitter.Spawn Rate)
  ]
  particleSpawnScript [
   &VC_PaintedVariantsOverride(
    Black=<rgba 0.2708 0.2708 0.2708 1>
    BurntSienna=<rgba 0.3 0.1108 0.06 1>
    Cobalt=<rgba 0.1771 0.1923 1 1>
    Crimson=<rgba 1 0.1562 0.1791 1>
    DefaultColor=<rgba 0.6376 0.6308 1 0.1>
    ForestGreen=<rgba 0.25 1 0.2879 1>
    Grey=<rgba 0.4323 0.4323 0.4323 1>
    Lime=<rgba 0.6 1 0.2 1>
    Orange=<rgba 1 0.4139 0.1 1>
    Pink=<rgba 1 0.2 0.7091 1>
    Purple=<rgba 0.5283 0.2 1 1>
    R1 Black:LinearColor=<rgba 0.03 0.03 0.03 1>
    R1 BurntSienna:LinearColor=<rgba 0.0642 0.0198 0.0079 1>
    R1 Cobalt:LinearColor=<rgba 0.0141 0.0194 0.3 1>
    R1 Crimson:LinearColor=<rgba 0.2 0 0.0054 1>
    R1 DefaultColor=<rgba 0.02 0.02 0.02 1>
    R1 ForestGreen:LinearColor=<rgba 0 0.3 0.0152 1>
    R1 Grey:LinearColor=<rgba 0.1615 0.1615 0.1615 1>
    R1 Lime:LinearColor=<rgba 0.15 0.3 0 1>
    R1 Orange:LinearColor=<rgba 0.5 0.1722 0 1>
    R2 Black:LinearColor=<rgba 0 0 0 1>
    R2 BurntSienna:LinearColor=<rgba 0.1094 0.0338 0.0135 1>
    R2 Cobalt:LinearColor=<rgba 0 0.0121 0.6562 1>
    R2 Crimson:LinearColor=<rgba 0.224 0 0.0061 1>
    R2 DefaultColor:LinearColor=<rgba 0 0 0 1>
    R2 ForestGreen:LinearColor=<rgba 0 0.25 0.0126 1>
    R2 Grey:LinearColor=<rgba 0.1667 0.1667 0.1667 1>
    Saffron=<rgba 1 0.8133 0.2 1>
    SkyBlue=<rgba 0 0.6654 1 1>
    Use R1 Painted Overrides=true
    Use R2 Painted Overrides=true
   ) {enabled=false}
   &VC_InitializeRibbon_v1(
    Custom Offset=true
    Enable Skeletal Position=true
    Facing Axis:ENiagaraOrientationAxis=::YAxis
    Lifetime=1
    Local Vector Offset=&AddVectorToPosition(
     Position=&MergeHelper_VectorSwitch(0=<-30 0 -34>, 1=<-30 0 -30>, 2=<-35 0 -25>, 3=?, Index=$System.GameModeID, Num=::NewEnumerator1)
     Vector=Linked() {linkedVariable {name=User.offset, type=NiagaraPosition}}
    )
    Offset Amount=1
    Ribbon Width=40
    Write Material Random=false
   )
   &AddVelocity(Velocity=<-1000 0 0>) {module=AddVelocity001}
   /SetVariables() {enabled=false, isSetParameters=true}
  ]
  particleUpdateScript [
   &FN_DisableTireFX(Mask Index Attribute:NiagaraInt32=$Particles.SpawnGroup, Mask Index FL=2, Mask Index FR=3, Mask Index RL=0, Mask Index RR=1)
   &ParticleState()
   &InheritSourceMovement(
    Applied Position Scale=&MergeHelper_VectorSwitch(
     0=&Multiply_VectorByFloat(Vector=<1 1 1>, Float=1)
     1=<1 1 1>
     2=<1 1 1>
     3=?
     Index=$System.GameModeID
     Num=::NewEnumerator1
    )
    ApplyForce=false
    ApplyPosition=true
    ApplyPositionFalloff=false
   )
   &SolveForcesAndVelocity(Clamp Velocity=false, Limit Acceleration=false)
   &VC_UpdateRibbon_v1(Enable Skeletal Position=true)
   &ScaleRibbonWidth(Ribbon Width Scale=&FloatFromCurve(
     FloatCurve=DataInterface() {propertyValues {Curve {keys [
         {interpMode=RCIM_Cubic, leaveTangent=1.3448, tangentMode=RCTM_Break, time=0, value=0.4}
         {arriveTangent=-0.0249, interpMode=RCIM_Cubic, leaveTangent=-0.0249, tangentMode=RCTM_User, time=0.1, value=0.9}
         {interpMode=RCIM_Cubic, tangentMode=RCTM_Break, time=1, value=0.3}
        ]}}}
     Scale Curve=&MergeHelper_FloatSwitch(0=1.2, 1=1.2, 2=1.2, 3=?, Index=$System.GameModeID, Num=::NewEnumerator1)
    ))
   /SetVariables(
    Particles.OilColor:LinearColor=&ColorFromCurve(ColorCurve=DataInterface() {propertyValues {
       AlphaCurve {keys [<at 0 0>,<at 0.1456 1 RCIM_Cubic>]}
       BlueCurve {keys [<at 0 0 RCIM_Cubic>,<at 1 0 RCIM_Cubic>]}
       GreenCurve {keys [<at 0 0 RCIM_Cubic>,<at 1 0 RCIM_Cubic>]}
       RedCurve {keys [<at 0 0 RCIM_Cubic>,<at 1 0 RCIM_Cubic>]}
      }})
    Particles.TirePrintsColor:LinearColor=&ColorFromCurve(ColorCurve=DataInterface() {propertyValues {
       AlphaCurve {keys [<at 0.2906 0 RCIM_Cubic>,<at 0.5112 1 RCIM_Cubic>]}
       BlueCurve {keys [<at 0 0 RCIM_Cubic>]}
       GreenCurve {keys [<at 0 0 RCIM_Cubic>]}
       RedCurve {keys [<at 0 0 RCIM_Cubic>]}
      }})
    Particles.WavesWidth:NiagaraFloat=&FloatFromCurve(FloatCurve=DataInterface() {propertyValues {Curve {keys [<at 0 0.9>,<at 1 0.6>]}}}, Scale Curve=39)
    Particles.Waves_Color:LinearColor=&ScaleLinearColorByCurve(
     InitialValue=<rgba 0.5586 0.5826 0.6875 0.08>
     ScaleCurve=DataInterface() {propertyValues {
       AlphaCurve {keys [<at 0.0406 0>,<at 0.0784 0.5>,<at 0.3303 1>,<at 0.6568 0>]}
       BlueCurve {keys [<at 0 0.599>,<at 0.8978 1>,<at 1 0.0625>]}
       GreenCurve {keys [<at 0 0.3757>,<at 0.8978 0.6308>,<at 1 0.0625>]}
       RedCurve {keys [<at 0 0.38>,<at 0.8978 0.6376>,<at 1 0.0625>]}
      }}
     CurveIndex=$Particles.NormalizedAge
    )
    Particles.WheelPrintsWidth=15
   ) {isSetParameters=true}
   &ScaleColor(
    Scale Alpha=&FloatFromCurve(FloatCurve=DataInterface() {propertyValues {Curve {keys [<at 0 0.5>,<at 0.6343 1>,<at 0.7343 0>]}}})
    Scale RGB=&VectorFromFloat(Value=&FloatFromCurve(FloatCurve=DataInterface() {propertyValues {Curve {keys [<at 0 0.5>,<at 0.5082 1>,<at 1 0.5>]}}}))
    ScaleA=true
    ScaleRGB=true
   )
  ]
  rendererClasses [&NiagaraRibbonRendererProperties]
  renderers [
   {rendererClass=&NiagaraRibbonRendererProperties}
   {rendererClass=&NiagaraRibbonRendererProperties, rendererIndex=1}
   {rendererClass=&NiagaraRibbonRendererProperties, rendererIndex=2}
  ]
 }
 NE_OilWheelWrap001 {
  emitterUpdateScript [
   &EmitterState(
    Inactive Response=::NewEnumerator1
    Life Cycle Mode=::NewEnumerator1
    Loop Behavior=::NewEnumerator1
    Loop Duration Mode=::NewEnumerator1
    UseLoopDelay=false
   )
   &SpawnBurst_Instantaneous(Spawn Count=2, Use Spawn Probability=false, UseLoopCountLimit=false)
  ]
  particleSpawnScript [
   &VC_PaintedVariantsOverride(
    BurntSienna=<rgba 0.025 0.0053 0 1>
    Cobalt=<rgba 0 0.0037 0.2 1>
    Crimson=<rgba 0.5 0 0.0136 1>
    DefaultColor=<rgba 0 0 0 1>
    ForestGreen=<rgba 0 0.3 0.0152 1>
    R1 Black:LinearColor=<rgba 0 0.0037 0.2 1>
    R1 Cobalt:LinearColor=<rgba 0.1 0.1 0.1 1>
    R1 DefaultColor=<rgba 0.3633 0.3633 0.5625 0.6>
    Use R1 Painted Overrides=true
   ) {enabled=false}
   &InitializeParticle(
    Color=<rgba 0 0 0 1>
    Mesh Scale Mode=::NewEnumerator4
    Sprite Rotation Mode=::NewEnumerator2
    Sprite Size Mode=::NewEnumerator3
    UsePositionOffset=false
    Write Material Random=true
   )
   &PartitionParticles()
   &SkeletalMeshLocation(
    Bone / Socket Index=&Add_Integer(A=$Particles.PartitionParticles.Partition, B=4)
    Bone / Socket Sampling Mode=::NewEnumerator5
    Interpolate Mesh Sampling=false
    MaskSpawnGroup=false
    Mesh Sampling Type=::NewEnumerator6
    Skeletal Mesh=$System.SkeletalMesh
   )
   /SetVariables(Particles.WheelsWavesColor:LinearColor=<rgba 0.1914 0.1991 0.3281 1>) {isSetParameters=true}
   &InitialMeshOrientation(Mesh Orientation Mode=::NewEnumerator7, Rotation=<0 0.24 0>, Use Rotation Vector=true)
  ]
  particleUpdateScript [
   &FN_DisableTireFX(
    Mask Index Attribute:NiagaraInt32=$Particles.PartitionParticles.Partition
    Mask Index FL=2
    Mask Index FR=3
    Mask Index RL=0
    Mask Index RR=1
   )
   &VC_WheelWrapScale(Wheel Index Attribute:NiagaraInt32=&Add_Integer(A=$Particles.PartitionParticles.Partition, B=2))
   &ParticleState(Kill Particles When Lifetime Has Elapsed=false, Let Infinitely Lived Particles Die When Emitter Deactivates=true)
   &UpdateMeshOrientation(Rotation Coordinate Space=::NewEnumerator2, Rotation Vector=<0 1 0>)
   &SkeletalMeshLocation(
    Bone / Socket Index=&Add_Integer(A=$Particles.PartitionParticles.Partition, B=4)
    Bone / Socket Sampling Mode=::NewEnumerator5
    MaskSpawnGroup=false
    Mesh Sampling Type=::NewEnumerator6
    Skeletal Mesh=$System.SkeletalMesh
   ) {module=SkeletalMeshLocation001}
   &SolveForcesAndVelocity(Clamp Velocity=false, Limit Acceleration=false)
  ]
  rendererClasses [&NiagaraMeshRendererProperties]
  renderers [
   {rendererClass=&NiagaraMeshRendererProperties, rendererIndex=0}
   {rendererClass=&NiagaraMeshRendererProperties, rendererIndex=1}
   {rendererClass=&NiagaraMeshRendererProperties, rendererIndex=2}
   {rendererClass=&NiagaraMeshRendererProperties, rendererIndex=3}
  ]
 }
}
summary {
 systemName=NS_DriftTrail_OilSlick
 emitters [
  {emitterName=NE_OilWheelWrap001, rendererClasses [&NiagaraMeshRendererProperties]}
  {
   emitterName=NE_Cone-Particles001
   rendererClasses [&NiagaraMeshRendererProperties, &NiagaraSpriteRendererProperties]
  }
  {emitterName=NE_OilTrail001, rendererClasses [&NiagaraRibbonRendererProperties]}
 ]
 userVariables [
  User.bIsFrontendPreview {type=NiagaraBool}
  User.bIsPainted {type=NiagaraBool, defaultValue=true}
  User.ColorVariantID {type=NiagaraInt32}
  User.GameModeID {type=NiagaraInt32}
  User.offset {type=NiagaraPosition, description="", defaultValue {}}
  User.PaintedColor {type=LinearColor, description="", defaultValue {struct=LinearColor, value=<rgba 1 1 1 1>}}
  User.PaintedScalar {type=NiagaraFloat, defaultValue=1, description=""}
  User.VehicleFlags {type=NiagaraInt32, defaultValue=806289408}
  User.Wheel_Radius_Array {
   type=NiagaraDataInterfaceArrayFloat
   defaultValue {
    struct=NiagaraExt_VariableValue_DataInterface
    value {
     dataInterface=&self:NiagaraDataInterfaceArrayFloat_1
     dataInterfaceClass=&NiagaraDataInterfaceArrayFloat
    }
   }
  }
  User.Wheel_Width_Array {
   type=NiagaraDataInterfaceArrayFloat
   defaultValue {
    struct=NiagaraExt_VariableValue_DataInterface
    value {
     dataInterface=&self:NiagaraDataInterfaceArrayFloat_0
     dataInterfaceClass=&NiagaraDataInterfaceArrayFloat
    }
   }
  }
  User.WheelTouchingArray {
   type=NiagaraDataInterfaceArrayFloat
   defaultValue {
    struct=NiagaraExt_VariableValue_DataInterface
    value {
     dataInterface=&self:NiagaraDataInterfaceArrayFloat_2
     dataInterfaceClass=&NiagaraDataInterfaceArrayFloat
    }
   }
  }
 ]
}
